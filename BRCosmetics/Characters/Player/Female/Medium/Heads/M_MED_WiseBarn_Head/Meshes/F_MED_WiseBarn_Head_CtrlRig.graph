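# Control Rig for the WiseBarn head cosmetic that applies a small corrective rotation to the chest ribbon bone (base_ribbon) while the character crouches and targets.
# A spherical pose reader on the left upper arm, relative to the clavicle, supplies the base weight.
# That weight is multiplied by the isTargetingCR and isCrouchingCR flags and by pelvis height above root, remapped through a rise-and-fall curve to limit the correction to a height band.
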ControlRigBlueprint /BRCosmetics/Characters/Player/Female/Medium/Heads/M_MED_WiseBarn_Head/Meshes/F_MED_WiseBarn_Head_CtrlRig
graphs [
 Multiply_1_ContainedGraph {nodes [
   n1 = RigVMFunctionEntryNode()
   Multiply = &RigVMFunction_MathDoubleMul(A=@n1.A, B=@n1.B)
   RigVMFunctionReturnNode(Result=&RigVMFunction_MathDoubleMul(B=@n1.C, A=@Multiply.Result).Result)
  ]}
 RigVMFunctionLibrary {links [], nodes []}
 RigVMModel {nodes [
   RigUnit_BeginExecution = &RigUnit_BeginExecution()
   n1 = &RigUnit_SphericalPoseReader(
    ActiveRegionScaleFactors=(PositiveWidth=2,NegativeWidth=2,PositiveHeight=2,NegativeHeight=2)
    ActiveRegionSize=0.05
    Debug=(bDrawDebug=false,bDraw2D=false,DebugScale=25,DebugSegments=20,DebugThickness=0.25)
    DriverAxis=(X=1,Y=0,Z=0)
    DriverItem=(Type=Bone,Name="dfrm_upperarm_l")
    FalloffRegionScaleFactors=(PositiveWidth=1,NegativeWidth=1,PositiveHeight=1,NegativeHeight=1)
    FalloffSize=0.2
    OptionalParentItem=(Type=Bone,Name="clavicle_l")
    RotationOffset=(X=110,Y=0,Z=-50)
    ExecutePin=@RigUnit_BeginExecution.ExecutePin
   )
   n2 = &RigVMFunction_AnimEvalRichCurve(
    Curve=(EditorCurveData=(Keys=((),(Time=0.5,Value=1),(Time=0.8))))
    SourceMaximum=2
    SourceMinimum=0
    TargetMinimum=0
    Value=&RigVMFunction_MathDoubleRemap(
     SourceMaximum=70
     SourceMinimum=45
     TargetMaximum=1
     TargetMinimum=0
     Value=&RigUnit_GetRelativeTransformForItem(Child=(Type=Bone,Name="pelvis"), Parent=(Type=Bone,Name="root"), bChildInitial=False, bParentInitial=False).RelativeTransform.Translation.Z
    ).Result
   )
   n3 = RigVMVariableNode() {variable=isCrouchingCR}
   RigVMFunction_MathBoolToFloat = &RigVMFunction_MathBoolToFloat(Value=RigVMVariableNode() {variable=isTargetingCR}.Value)
   Multiply = RigVMAggregateNode(
    C=&RigVMFunction_MathDoubleMul(B=@n2.Result, A=&RigVMFunction_MathBoolToFloat(Value=@n3.Value).Result).Result
    B=@RigVMFunction_MathBoolToFloat.Result
    A=@n1.OutputParam
   )
   > &RigUnit_ModifyTransforms(
    ItemToModify=((Item=(Type=Bone,Name="base_ribbon"),Transform=(Rotation=(X=0.0872,Y=0,Z=0,W=0.9962),Translation=(X=0,Y=0,Z=0),Scale3D=(X=1,Y=1,Z=1))))
    WeightMinimum=0
    Weight=@Multiply.Result
   )
  ]}
]
hierarchy [
 root {loc=<0 0 0>, rot=<0 0 0 1>}
 pelvis {parent=root, loc=<0 -0.5301 98.0571>, rot=<0.0127 0.707 -0.0127 -0.707>}
 spine_01 {parent=pelvis, loc=<6.1208 -0.9328 0>, rot=<0 0 0.034 -0.9994>}
 spine_02 {parent=spine_01, loc=<6.5997 -0.2095 0>, rot=<0 0 0.0238 -0.9997>}
 spine_03 {parent=spine_02, loc=<6.4914 -0.1237 0>, rot=<0 0 0.0149 -0.9999>}
 spine_04 {parent=spine_03, loc=<6.9598 0.0794 0>, rot=<0 0 -0.0132 -0.9999>}
 spine_05 {parent=spine_04, loc=<6.9987 0 0>, rot=<0 0 -0.0015 -1>}
 neck_01 {parent=spine_05, loc=<11.3138 0.1529 0>, rot=<0 0 -0.2003 -0.9797>}
 neck_02 {parent=neck_01, loc=<4.5627 0 0>, rot=<0 0 0.0504 -0.9987>}
 head {parent=neck_02, loc=<5.3914 0 0>, rot=<0 0 0.115 -0.9934>}
 base_hat {parent=head, loc=<16.5304 -4.4277 -0.3522>, rot=<0.8922 -0.2837 -0.0236 -0.3506>}
 dyn_hay_01 {parent=base_hat, loc=<8.7176 -0.0533 0.0095>, rot=<-0.1049 0.0508 -0.4385 -0.8911>}
 dyn_hay_02 {parent=dyn_hay_01, loc=<11.0374 -0.0785 -0.0229>, rot=<-0.0678 -0.0152 -0.4477 -0.8915>}
 dyn_hay_03 {parent=dyn_hay_02, loc=<7.4694 -0.0191 -0.0504>, rot=<0.2163 -0.1286 -0.3343 -0.9083>}
 dyn_hay_04 {parent=dyn_hay_03, loc=<4.2909 -0.0042 0.033>, rot=<0 0 -0.0552 -0.9985>}
 faceAttach {parent=head, loc=<-4.146 -0.8545 0>, rot=<0.0029 -0.7071 -0.0029 -0.7071>}
 C_brow_mid {parent=faceAttach, loc=<0 15.1889 14.4883>, rot=<-0.1256 0 0 -0.9921>}
 L_brow_mid {parent=faceAttach, loc=<7.4651 14.6986 16.9393>, rot=<0.5742 0.2007 0.3622 -0.7062>}
 L_brow_outer {parent=faceAttach, loc=<13.4254 8.2954 21.1702>, rot=<0.1182 -0.0284 0.3942 -0.9109>}
 L_cheek_inner {parent=faceAttach, loc=<11.4576 14.472 1.5663>, rot=<0.0234 0.0036 0.4071 -0.9131>}
 C_lip_upper_mid {parent=faceAttach, loc=<0.0005 18.7719 13.6887>, rot=<-0.0477 0.0011 -0.0008 -0.9989>}
 L_lip_upper_outer {parent=faceAttach, loc=<1.7842 17.4192 9.4158>, rot=<-0.226 0.0512 0.5662 -0.791>}
 L_lip_corner {parent=faceAttach, loc=<2.6162 15.4247 9.1287>, rot=<-0.0018 0.0083 0.5289 -0.8487>}
 R_brow_mid {parent=faceAttach, loc=<-7.4651 14.6986 16.9393>, rot=<0.5742 -0.2007 -0.3622 -0.7062>}
 R_brow_outer {parent=faceAttach, loc=<-13.4254 8.2954 21.1702>, rot=<0.1182 0.0284 -0.3942 -0.9109>}
 R_cheek_inner {parent=faceAttach, loc=<-11.4576 14.472 1.5663>, rot=<0.0234 -0.0036 -0.4071 -0.9131>}
 R_lip_corner {parent=faceAttach, loc=<-2.6162 15.4247 9.1287>, rot=<-0.0018 -0.0083 -0.5289 -0.8487>}
 C_jaw {parent=faceAttach, loc=<0 14.0908 9.5719>, rot=<0 0 0 1>}
 L_lip_lower_outer {parent=C_jaw, loc=<1.2878 2.993 0.8819>, rot=<-0.0486 -0.1481 0.5583 -0.8149>}
 C_lip_lower_mid {parent=C_jaw, loc=<0.061 5.5123 -0.0035>, rot=<-0.4409 0 0 -0.8975>}
 R_lip_lower_outer {parent=C_jaw, loc=<-1.2878 2.993 0.8819>, rot=<-0.0486 0.1481 -0.5583 -0.8149>}
 tongue {parent=C_jaw, loc=<0.0033 1.9634 -0.5996>, rot=<-0.3709 0 0 -0.9287>}
 R_eye {parent=faceAttach, loc=<-8.0084 9.6592 12.7996>, rot=<0.22 -0.1294 0.9641 -0.0731>}
 R_eye_lid_upper_mid {parent=R_eye, loc=<0 0 0>, rot=<0.0523 0 0 -0.9986>}
 R_eye_lid_lower_mid {parent=R_eye, loc=<0 0 0>, rot=<0.0329 0 0 -0.9995>}
 L_eye {parent=faceAttach, loc=<8.0084 9.6592 12.7995>, rot=<-0.22 -0.1294 0.9641 0.0731>}
 L_eye_lid_upper_mid {parent=L_eye, loc=<0 0 0>, rot=<0.0523 0 0 -0.9986>}
 L_eye_lid_lower_mid {parent=L_eye, loc=<0 0 0>, rot=<0.0329 0 0 -0.9995>}
 R_lip_upper_outer {parent=faceAttach, loc=<-1.7842 17.4192 9.4158>, rot=<-0.226 -0.0512 -0.5662 -0.791>}
 clavicle_l {parent=spine_05, loc=<5.4831 -0.1053 -3.2946>, rot=<0.0465 -0.7289 -0.009 -0.6829>}
 clavicle_r {parent=spine_05, loc=<5.4831 -0.1053 3.2946>, rot=<0.7289 0.0465 -0.6829 0.009>}
 dfrm_clavicle_l {parent=spine_05, loc=<5.4831 -0.1053 -3.2946>, rot=<0.0465 -0.7289 -0.009 -0.6829>}
 dfrm_upperarm_l {parent=dfrm_clavicle_l, loc=<15.6259 -0.8551 -0.145>, rot=<0.0185 -0.3968 -0.0255 -0.9174>}
 dfrm_upperarm_twist_01_l {parent=dfrm_upperarm_l, loc=<7.2135 -0.1134 0>, rot=<0.0015 0 0 -1>}
 dfrm_clavicle_r {parent=spine_05, loc=<5.4831 -0.1053 3.2946>, rot=<0.7289 0.0465 -0.6829 0.009>}
 dfrm_upperarm_r {parent=dfrm_clavicle_r, loc=<-15.6259 0.8551 0.145>, rot=<0.0184 -0.396 -0.0255 -0.9177>}
 dfrm_upperarm_twist_01_r {parent=dfrm_upperarm_r, loc=<-7.2135 0.1134 0>, rot=<0 0 0 -1>}
 base_ribbon {parent=spine_05, loc=<4.4979 18.4328 0>, rot=<0.0284 -0.7065 -0.0284 -0.7065>}
 dyn_ribbon {parent=base_ribbon, loc=<0 0 0>, rot=<0 0 0 1>}
 dyn_ribbon_r {parent=dyn_ribbon, loc=<-1.6826 0.1238 0.0045>, rot=<-0.2505 -0.7765 0.5702 -0.0957>}
 dyn_ribbon_l {parent=dyn_ribbon, loc=<1.6826 0.1238 0.0045>, rot=<0.5693 -0.1088 -0.2526 -0.7748>}
 dyn_medal {parent=dyn_ribbon, loc=<0.0709 0.4483 -1.7865>, rot=<0.3288 -0.3288 -0.626 -0.626>}
 look_left_pose {type=Curve, loc=<0 0 0>, rot=<0 0 0 1>}
 look_right_pose {type=Curve, loc=<0 0 0>, rot=<0 0 0 1>}
 look_up_pose {type=Curve, loc=<0 0 0>, rot=<0 0 0 1>}
 look_down_pose {type=Curve, loc=<0 0 0>, rot=<0 0 0 1>}
 L_blink_pose {type=Curve, loc=<0 0 0>, rot=<0 0 0 1>}
 R_blink_pose {type=Curve, loc=<0 0 0>, rot=<0 0 0 1>}
 L_wide_pose {type=Curve, loc=<0 0 0>, rot=<0 0 0 1>}
 R_wide_pose {type=Curve, loc=<0 0 0>, rot=<0 0 0 1>}
 R_squint_inner_pose {type=Curve, loc=<0 0 0>, rot=<0 0 0 1>}
 R_squeeze_pose {type=Curve, loc=<0 0 0>, rot=<0 0 0 1>}
 L_squint_inner_pose {type=Curve, loc=<0 0 0>, rot=<0 0 0 1>}
 L_squeeze_pose {type=Curve, loc=<0 0 0>, rot=<0 0 0 1>}
 L_brow_up_pose {type=Curve, loc=<0 0 0>, rot=<0 0 0 1>}
 L_brow_down_pose {type=Curve, loc=<0 0 0>, rot=<0 0 0 1>}
 R_brow_up_pose {type=Curve, loc=<0 0 0>, rot=<0 0 0 1>}
 R_brow_down_pose {type=Curve, loc=<0 0 0>, rot=<0 0 0 1>}
 C_glabella_up_pose {type=Curve, loc=<0 0 0>, rot=<0 0 0 1>}
 C_glabella_down_pose {type=Curve, loc=<0 0 0>, rot=<0 0 0 1>}
 L_brow_squeeze_in_pose {type=Curve, loc=<0 0 0>, rot=<0 0 0 1>}
 R_brow_squeeze_in_pose {type=Curve, loc=<0 0 0>, rot=<0 0 0 1>}
 L_upper_lip_raiser_pose {type=Curve, loc=<0 0 0>, rot=<0 0 0 1>}
 R_upper_lip_raiser_pose {type=Curve, loc=<0 0 0>, rot=<0 0 0 1>}
 L_upper_lip_lower_pose {type=Curve, loc=<0 0 0>, rot=<0 0 0 1>}
 R_upper_lip_lower_pose {type=Curve, loc=<0 0 0>, rot=<0 0 0 1>}
 L_lower_lip_up_pose {type=Curve, loc=<0 0 0>, rot=<0 0 0 1>}
 R_lower_lip_up_pose {type=Curve, loc=<0 0 0>, rot=<0 0 0 1>}
 L_lower_lip_down_pose {type=Curve, loc=<0 0 0>, rot=<0 0 0 1>}
 R_lower_lip_down_pose {type=Curve, loc=<0 0 0>, rot=<0 0 0 1>}
 L_lip_corner_narrow_pose {type=Curve, loc=<0 0 0>, rot=<0 0 0 1>}
 L_lip_corner_wide_pose {type=Curve, loc=<0 0 0>, rot=<0 0 0 1>}
 R_lip_corner_narrow_pose {type=Curve, loc=<0 0 0>, rot=<0 0 0 1>}
 R_lip_corner_wide_pose {type=Curve, loc=<0 0 0>, rot=<0 0 0 1>}
 L_smile_pose {type=Curve, loc=<0 0 0>, rot=<0 0 0 1>}
 R_smile_pose {type=Curve, loc=<0 0 0>, rot=<0 0 0 1>}
 L_frown_pose {type=Curve, loc=<0 0 0>, rot=<0 0 0 1>}
 R_frown_pose {type=Curve, loc=<0 0 0>, rot=<0 0 0 1>}
 phoneme_oo_pose {type=Curve, loc=<0 0 0>, rot=<0 0 0 1>}
 phoneme_ch_pose {type=Curve, loc=<0 0 0>, rot=<0 0 0 1>}
 phoneme_mbp_pose {type=Curve, loc=<0 0 0>, rot=<0 0 0 1>}
 phoneme_fv_pose {type=Curve, loc=<0 0 0>, rot=<0 0 0 1>}
 mouth_up_pose {type=Curve, loc=<0 0 0>, rot=<0 0 0 1>}
 mouth_down_pose {type=Curve, loc=<0 0 0>, rot=<0 0 0 1>}
 C_nose_wrinkler_pose {type=Curve, loc=<0 0 0>, rot=<0 0 0 1>}
 L_cheek_up_pose {type=Curve, loc=<0 0 0>, rot=<0 0 0 1>}
 R_cheek_up_pose {type=Curve, loc=<0 0 0>, rot=<0 0 0 1>}
 jaw_open_pose {type=Curve, loc=<0 0 0>, rot=<0 0 0 1>}
 jaw_right_pose {type=Curve, loc=<0 0 0>, rot=<0 0 0 1>}
 jaw_left_pose {type=Curve, loc=<0 0 0>, rot=<0 0 0 1>}
 eye_convergence_pose {type=Curve, loc=<0 0 0>, rot=<0 0 0 1>}
 L_upperLidDown_pose {type=Curve, loc=<0 0 0>, rot=<0 0 0 1>}
 R_upperLidDown_pose {type=Curve, loc=<0 0 0>, rot=<0 0 0 1>}
 jaw_up_pose {type=Curve, loc=<0 0 0>, rot=<0 0 0 1>}
 jaw_forward_pose {type=Curve, loc=<0 0 0>, rot=<0 0 0 1>}
 jaw_back_pose {type=Curve, loc=<0 0 0>, rot=<0 0 0 1>}
 tongue_forward_pose {type=Curve, loc=<0 0 0>, rot=<0 0 0 1>}
 tongue_back_pose {type=Curve, loc=<0 0 0>, rot=<0 0 0 1>}
 tongue_up_pose {type=Curve, loc=<0 0 0>, rot=<0 0 0 1>}
 tongue_down_pose {type=Curve, loc=<0 0 0>, rot=<0 0 0 1>}
 base_pose {type=Curve, loc=<0 0 0>, rot=<0 0 0 1>}
 Hair_Scarf {type=Curve, loc=<0 0 0>, rot=<0 0 0 1>}
 Hair_Hat {type=Curve, loc=<0 0 0>, rot=<0 0 0 1>}
 Hair_Helmet {type=Curve, loc=<0 0 0>, rot=<0 0 0 1>}
 Hair_Glasses {type=Curve, loc=<0 0 0>, rot=<0 0 0 1>}
 Hair_Cap {type=Curve, loc=<0 0 0>, rot=<0 0 0 1>}
 Hair_Headsets {type=Curve, loc=<0 0 0>, rot=<0 0 0 1>}
 Hair_Hood {type=Curve, loc=<0 0 0>, rot=<0 0 0 1>}
 Face_Mask {type=Curve, loc=<0 0 0>, rot=<0 0 0 1>}
 HairStrands_Hat {type=Curve, loc=<0 0 0>, rot=<0 0 0 1>}
 Hat_Hair {type=Curve, loc=<0 0 0>, rot=<0 0 0 1>}
 Helmet_Hair {type=Curve, loc=<0 0 0>, rot=<0 0 0 1>}
 Hair_Muff {type=Curve, loc=<0 0 0>, rot=<0 0 0 1>}
 Hair_SpaceSuit {type=Curve, loc=<0 0 0>, rot=<0 0 0 1>}
 Hair_Hide {type=Curve, loc=<0 0 0>, rot=<0 0 0 1>}
 Head_SpaceSuit {type=Curve, loc=<0 0 0>, rot=<0 0 0 1>}
 Hair_Null {type=Curve, loc=<0 0 0>, rot=<0 0 0 1>}
 DisableFaceOverride {type=Curve, loc=<0 0 0>, rot=<0 0 0 1>}
 LiveLink {type=Curve, loc=<0 0 0>, rot=<0 0 0 1>}
 Hat_MESH {type=Curve, loc=<0 0 0>, rot=<0 0 0 1>}
 Null_MESH {type=Curve, loc=<0 0 0>, rot=<0 0 0 1>}
 Pose_0 {type=Curve, loc=<0 0 0>, rot=<0 0 0 1>}
 Hair_Mask {type=Curve, loc=<0 0 0>, rot=<0 0 0 1>}
 eye_convergence {type=Curve, loc=<0 0 0>, rot=<0 0 0 1>}
 headTransform {type=Curve, loc=<0 0 0>, rot=<0 0 0 1>}
 Pose_1 {type=Curve, loc=<0 0 0>, rot=<0 0 0 1>}
 Pose_2 {type=Curve, loc=<0 0 0>, rot=<0 0 0 1>}
 Pose_3 {type=Curve, loc=<0 0 0>, rot=<0 0 0 1>}
 Pose_4 {type=Curve, loc=<0 0 0>, rot=<0 0 0 1>}
 Pose_5 {type=Curve, loc=<0 0 0>, rot=<0 0 0 1>}
 Pose_6 {type=Curve, loc=<0 0 0>, rot=<0 0 0 1>}
 Pose_7 {type=Curve, loc=<0 0 0>, rot=<0 0 0 1>}
 Pose_8 {type=Curve, loc=<0 0 0>, rot=<0 0 0 1>}
 Pose_9 {type=Curve, loc=<0 0 0>, rot=<0 0 0 1>}
 Pose_10 {type=Curve, loc=<0 0 0>, rot=<0 0 0 1>}
 Pose_11 {type=Curve, loc=<0 0 0>, rot=<0 0 0 1>}
 Pose_12 {type=Curve, loc=<0 0 0>, rot=<0 0 0 1>}
 Pose_13 {type=Curve, loc=<0 0 0>, rot=<0 0 0 1>}
 Pose_14 {type=Curve, loc=<0 0 0>, rot=<0 0 0 1>}
 Pose_15 {type=Curve, loc=<0 0 0>, rot=<0 0 0 1>}
 Pose_16 {type=Curve, loc=<0 0 0>, rot=<0 0 0 1>}
 Pose_17 {type=Curve, loc=<0 0 0>, rot=<0 0 0 1>}
 Pose_18 {type=Curve, loc=<0 0 0>, rot=<0 0 0 1>}
 Pose_19 {type=Curve, loc=<0 0 0>, rot=<0 0 0 1>}
 Pose_20 {type=Curve, loc=<0 0 0>, rot=<0 0 0 1>}
 Pose_21 {type=Curve, loc=<0 0 0>, rot=<0 0 0 1>}
 Pose_22 {type=Curve, loc=<0 0 0>, rot=<0 0 0 1>}
 Pose_23 {type=Curve, loc=<0 0 0>, rot=<0 0 0 1>}
 Pose_24 {type=Curve, loc=<0 0 0>, rot=<0 0 0 1>}
 Pose_25 {type=Curve, loc=<0 0 0>, rot=<0 0 0 1>}
 Pose_26 {type=Curve, loc=<0 0 0>, rot=<0 0 0 1>}
 Pose_27 {type=Curve, loc=<0 0 0>, rot=<0 0 0 1>}
 Pose_28 {type=Curve, loc=<0 0 0>, rot=<0 0 0 1>}
 Pose_29 {type=Curve, loc=<0 0 0>, rot=<0 0 0 1>}
 Pose_30 {type=Curve, loc=<0 0 0>, rot=<0 0 0 1>}
 Pose_31 {type=Curve, loc=<0 0 0>, rot=<0 0 0 1>}
 Pose_32 {type=Curve, loc=<0 0 0>, rot=<0 0 0 1>}
 Pose_33 {type=Curve, loc=<0 0 0>, rot=<0 0 0 1>}
 Pose_34 {type=Curve, loc=<0 0 0>, rot=<0 0 0 1>}
 Pose_35 {type=Curve, loc=<0 0 0>, rot=<0 0 0 1>}
 Pose_36 {type=Curve, loc=<0 0 0>, rot=<0 0 0 1>}
 Pose_37 {type=Curve, loc=<0 0 0>, rot=<0 0 0 1>}
 Pose_38 {type=Curve, loc=<0 0 0>, rot=<0 0 0 1>}
 Pose_39 {type=Curve, loc=<0 0 0>, rot=<0 0 0 1>}
 Pose_40 {type=Curve, loc=<0 0 0>, rot=<0 0 0 1>}
 Pose_41 {type=Curve, loc=<0 0 0>, rot=<0 0 0 1>}
 Pose_42 {type=Curve, loc=<0 0 0>, rot=<0 0 0 1>}
 Pose_43 {type=Curve, loc=<0 0 0>, rot=<0 0 0 1>}
 Pose_44 {type=Curve, loc=<0 0 0>, rot=<0 0 0 1>}
 Pose_45 {type=Curve, loc=<0 0 0>, rot=<0 0 0 1>}
 Pose_46 {type=Curve, loc=<0 0 0>, rot=<0 0 0 1>}
 Pose_47 {type=Curve, loc=<0 0 0>, rot=<0 0 0 1>}
 Pose_48 {type=Curve, loc=<0 0 0>, rot=<0 0 0 1>}
 Pose_49 {type=Curve, loc=<0 0 0>, rot=<0 0 0 1>}
 Pose_50 {type=Curve, loc=<0 0 0>, rot=<0 0 0 1>}
 Pose_51 {type=Curve, loc=<0 0 0>, rot=<0 0 0 1>}
 Pose_52 {type=Curve, loc=<0 0 0>, rot=<0 0 0 1>}
 Pose_53 {type=Curve, loc=<0 0 0>, rot=<0 0 0 1>}
 Pose_54 {type=Curve, loc=<0 0 0>, rot=<0 0 0 1>}
 Pose_55 {type=Curve, loc=<0 0 0>, rot=<0 0 0 1>}
 Pose_56 {type=Curve, loc=<0 0 0>, rot=<0 0 0 1>}
 Pose_57 {type=Curve, loc=<0 0 0>, rot=<0 0 0 1>}
 Pose_58 {type=Curve, loc=<0 0 0>, rot=<0 0 0 1>}
 Pose_59 {type=Curve, loc=<0 0 0>, rot=<0 0 0 1>}
 Pose_60 {type=Curve, loc=<0 0 0>, rot=<0 0 0 1>}
 Pose_61 {type=Curve, loc=<0 0 0>, rot=<0 0 0 1>}
 Pose_62 {type=Curve, loc=<0 0 0>, rot=<0 0 0 1>}
 Pose_63 {type=Curve, loc=<0 0 0>, rot=<0 0 0 1>}
 Pose_64 {type=Curve, loc=<0 0 0>, rot=<0 0 0 1>}
 Pose_65 {type=Curve, loc=<0 0 0>, rot=<0 0 0 1>}
 Pose_66 {type=Curve, loc=<0 0 0>, rot=<0 0 0 1>}
 Pose_67 {type=Curve, loc=<0 0 0>, rot=<0 0 0 1>}
 Pose_68 {type=Curve, loc=<0 0 0>, rot=<0 0 0 1>}
 Pose_69 {type=Curve, loc=<0 0 0>, rot=<0 0 0 1>}
 eyeLookDownRight {type=Curve, loc=<0 0 0>, rot=<0 0 0 1>}
 eyeLookOutRight {type=Curve, loc=<0 0 0>, rot=<0 0 0 1>}
 eyeLookUpRight {type=Curve, loc=<0 0 0>, rot=<0 0 0 1>}
 eyeLookOutLeft {type=Curve, loc=<0 0 0>, rot=<0 0 0 1>}
 eyeBlinkLeft {type=Curve, loc=<0 0 0>, rot=<0 0 0 1>}
 eyeBlinkRight {type=Curve, loc=<0 0 0>, rot=<0 0 0 1>}
 eyeSquintLeft {type=Curve, loc=<0 0 0>, rot=<0 0 0 1>}
 eyeSquintRight {type=Curve, loc=<0 0 0>, rot=<0 0 0 1>}
 eyeWideLeft {type=Curve, loc=<0 0 0>, rot=<0 0 0 1>}
 eyeWideRight {type=Curve, loc=<0 0 0>, rot=<0 0 0 1>}
 jawForward {type=Curve, loc=<0 0 0>, rot=<0 0 0 1>}
 jawLeft {type=Curve, loc=<0 0 0>, rot=<0 0 0 1>}
 jawRight {type=Curve, loc=<0 0 0>, rot=<0 0 0 1>}
 jawOpen {type=Curve, loc=<0 0 0>, rot=<0 0 0 1>}
 mouthClose {type=Curve, loc=<0 0 0>, rot=<0 0 0 1>}
 mouthFunnel {type=Curve, loc=<0 0 0>, rot=<0 0 0 1>}
 mouthPucker {type=Curve, loc=<0 0 0>, rot=<0 0 0 1>}
 mouthLeft {type=Curve, loc=<0 0 0>, rot=<0 0 0 1>}
 mouthRight {type=Curve, loc=<0 0 0>, rot=<0 0 0 1>}
 mouthSmileLeft {type=Curve, loc=<0 0 0>, rot=<0 0 0 1>}
 mouthSmileRight {type=Curve, loc=<0 0 0>, rot=<0 0 0 1>}
 mouthFrownLeft {type=Curve, loc=<0 0 0>, rot=<0 0 0 1>}
 mouthFrownRight {type=Curve, loc=<0 0 0>, rot=<0 0 0 1>}
 mouthDimpleLeft {type=Curve, loc=<0 0 0>, rot=<0 0 0 1>}
 mouthDimpleRight {type=Curve, loc=<0 0 0>, rot=<0 0 0 1>}
 mouthStretchLeft {type=Curve, loc=<0 0 0>, rot=<0 0 0 1>}
 mouthStretchRight {type=Curve, loc=<0 0 0>, rot=<0 0 0 1>}
 mouthRollLower {type=Curve, loc=<0 0 0>, rot=<0 0 0 1>}
 mouthRollUpper {type=Curve, loc=<0 0 0>, rot=<0 0 0 1>}
 mouthShrugLower {type=Curve, loc=<0 0 0>, rot=<0 0 0 1>}
 mouthShrugUpper {type=Curve, loc=<0 0 0>, rot=<0 0 0 1>}
 mouthPressLeft {type=Curve, loc=<0 0 0>, rot=<0 0 0 1>}
 mouthPressRight {type=Curve, loc=<0 0 0>, rot=<0 0 0 1>}
 mouthLowerDownLeft {type=Curve, loc=<0 0 0>, rot=<0 0 0 1>}
 mouthLowerDownRight {type=Curve, loc=<0 0 0>, rot=<0 0 0 1>}
 mouthUpperUpLeft {type=Curve, loc=<0 0 0>, rot=<0 0 0 1>}
 mouthUpperUpRight {type=Curve, loc=<0 0 0>, rot=<0 0 0 1>}
 browDownLeft {type=Curve, loc=<0 0 0>, rot=<0 0 0 1>}
 browDownRight {type=Curve, loc=<0 0 0>, rot=<0 0 0 1>}
 browInnerUp {type=Curve, loc=<0 0 0>, rot=<0 0 0 1>}
 browOuterUpLeft {type=Curve, loc=<0 0 0>, rot=<0 0 0 1>}
 browOuterUpRight {type=Curve, loc=<0 0 0>, rot=<0 0 0 1>}
 cheekPuff {type=Curve, loc=<0 0 0>, rot=<0 0 0 1>}
 cheekSquintLeft {type=Curve, loc=<0 0 0>, rot=<0 0 0 1>}
 cheekSquintRight {type=Curve, loc=<0 0 0>, rot=<0 0 0 1>}
 noseSneerLeft {type=Curve, loc=<0 0 0>, rot=<0 0 0 1>}
 noseSneerRight {type=Curve, loc=<0 0 0>, rot=<0 0 0 1>}
 C_glabella_down {type=Curve, loc=<0 0 0>, rot=<0 0 0 1>}
 L_brow_squeeze_in {type=Curve, loc=<0 0 0>, rot=<0 0 0 1>}
 R_brow_squeeze_in {type=Curve, loc=<0 0 0>, rot=<0 0 0 1>}
 L_upper_lip_lower {type=Curve, loc=<0 0 0>, rot=<0 0 0 1>}
 R_upper_lip_lower {type=Curve, loc=<0 0 0>, rot=<0 0 0 1>}
 L_lower_lip_up {type=Curve, loc=<0 0 0>, rot=<0 0 0 1>}
 R_lower_lip_up {type=Curve, loc=<0 0 0>, rot=<0 0 0 1>}
 L_lip_corner_narrow {type=Curve, loc=<0 0 0>, rot=<0 0 0 1>}
 R_lip_corner_narrow {type=Curve, loc=<0 0 0>, rot=<0 0 0 1>}
 phoneme_ch {type=Curve, loc=<0 0 0>, rot=<0 0 0 1>}
 phoneme_mbp {type=Curve, loc=<0 0 0>, rot=<0 0 0 1>}
 phoneme_fv {type=Curve, loc=<0 0 0>, rot=<0 0 0 1>}
 mouth_up {type=Curve, loc=<0 0 0>, rot=<0 0 0 1>}
 mouth_down {type=Curve, loc=<0 0 0>, rot=<0 0 0 1>}
 L_cheek_up {type=Curve, loc=<0 0 0>, rot=<0 0 0 1>}
 R_cheek_up {type=Curve, loc=<0 0 0>, rot=<0 0 0 1>}
 L_upperLidDown {type=Curve, loc=<0 0 0>, rot=<0 0 0 1>}
 R_upperLidDown {type=Curve, loc=<0 0 0>, rot=<0 0 0 1>}
 jaw_up {type=Curve, loc=<0 0 0>, rot=<0 0 0 1>}
 jaw_back {type=Curve, loc=<0 0 0>, rot=<0 0 0 1>}
 tongue_forward {type=Curve, loc=<0 0 0>, rot=<0 0 0 1>}
 tongue_back {type=Curve, loc=<0 0 0>, rot=<0 0 0 1>}
 tongue_up {type=Curve, loc=<0 0 0>, rot=<0 0 0 1>}
 tongue_down {type=Curve, loc=<0 0 0>, rot=<0 0 0 1>}
 basePose {type=Curve, loc=<0 0 0>, rot=<0 0 0 1>}
 neutral {type=Curve, loc=<0 0 0>, rot=<0 0 0 1>}
 cheek_puff_low {type=Curve, loc=<0 0 0>, rot=<0 0 0 1>}
 Desert_Island_Jonesy_Mouth {type=Curve, loc=<0 0 0>, rot=<0 0 0 1>}
 HT_L1_TransitionCurve {type=Curve, loc=<0 0 0>, rot=<0 0 0 1>}
 Hair_CapShape {type=Curve, loc=<0 0 0>, rot=<0 0 0 1>}
 Hair_CapShape1 {type=Curve, loc=<0 0 0>, rot=<0 0 0 1>}
 jekyll {type=Curve, loc=<0 0 0>, rot=<0 0 0 1>}
 JH_TransformCurve {type=Curve, loc=<0 0 0>, rot=<0 0 0 1>}
 eyes_LOD0 {type=Curve, loc=<0 0 0>, rot=<0 0 0 1>}
 MaterialCurveA {type=Curve, loc=<0 0 0>, rot=<0 0 0 1>}
 MaterialCurveB {type=Curve, loc=<0 0 0>, rot=<0 0 0 1>}
 MaterialCurveC {type=Curve, loc=<0 0 0>, rot=<0 0 0 1>}
 MaterialCurveD {type=Curve, loc=<0 0 0>, rot=<0 0 0 1>}
 MaterialCurveE {type=Curve, loc=<0 0 0>, rot=<0 0 0 1>}
 Custom_Frontend_Face_Idle {type=Curve, loc=<0 0 0>, rot=<0 0 0 1>}
 blendParent1 {type=Curve, loc=<0 0 0>, rot=<0 0 0 1>}
 BaseHeadDefaultFaceIdle {type=Curve, loc=<0 0 0>, rot=<0 0 0 1>}
 VeiledTransitionGlowAmount {type=Curve, loc=<0 0 0>, rot=<0 0 0 1>}
 is_3L {type=Curve, loc=<0 0 0>, rot=<0 0 0 1>}
 Hair_Hat_LOD0 {type=Curve, loc=<0 0 0>, rot=<0 0 0 1>}
 Hair_Helmet_LOD0 {type=Curve, loc=<0 0 0>, rot=<0 0 0 1>}
 Hair_Null_LOD0 {type=Curve, loc=<0 0 0>, rot=<0 0 0 1>}
 mouth_close_pose {type=Curve, loc=<0 0 0>, rot=<0 0 0 1>}
 mouth_pucker_pose {type=Curve, loc=<0 0 0>, rot=<0 0 0 1>}
 mouth_left_pose {type=Curve, loc=<0 0 0>, rot=<0 0 0 1>}
 mouth_right_pose {type=Curve, loc=<0 0 0>, rot=<0 0 0 1>}
 L_mouth_stretch_pose {type=Curve, loc=<0 0 0>, rot=<0 0 0 1>}
 R_mouth_stretch_pose {type=Curve, loc=<0 0 0>, rot=<0 0 0 1>}
 mouth_roll_lower_pose {type=Curve, loc=<0 0 0>, rot=<0 0 0 1>}
 mouth_roll_upper_pose {type=Curve, loc=<0 0 0>, rot=<0 0 0 1>}
 mouth_shrug_lower_pose {type=Curve, loc=<0 0 0>, rot=<0 0 0 1>}
 mouth_shrug_upper_pose {type=Curve, loc=<0 0 0>, rot=<0 0 0 1>}
 L_mouth_press_pose {type=Curve, loc=<0 0 0>, rot=<0 0 0 1>}
 R_mouth_press_pose {type=Curve, loc=<0 0 0>, rot=<0 0 0 1>}
 cheek_puff_pose {type=Curve, loc=<0 0 0>, rot=<0 0 0 1>}
 L_nose_wrinkler_pose {type=Curve, loc=<0 0 0>, rot=<0 0 0 1>}
 R_nose_wrinkler_pose {type=Curve, loc=<0 0 0>, rot=<0 0 0 1>}
 base {type=Curve, loc=<0 0 0>, rot=<0 0 0 1>}
 eyeLookDownLeft {type=Curve, loc=<0 0 0>, rot=<0 0 0 1>}
 eyeLookInLeft {type=Curve, loc=<0 0 0>, rot=<0 0 0 1>}
 eyeLookUpLeft {type=Curve, loc=<0 0 0>, rot=<0 0 0 1>}
 eyeLookInRight {type=Curve, loc=<0 0 0>, rot=<0 0 0 1>}
 is_face_bone_anim {type=Curve, loc=<0 0 0>, rot=<0 0 0 1>}
 FullHourglass {type=Curve, loc=<0 0 0>, rot=<0 0 0 1>}
 LeanRectangle {type=Curve, loc=<0 0 0>, rot=<0 0 0 1>}
 InvertedTriangle {type=Curve, loc=<0 0 0>, rot=<0 0 0 1>}
 L_eye {type=Curve, loc=<0 0 0>, rot=<0 0 0 1>}
 R_eye {type=Curve, loc=<0 0 0>, rot=<0 0 0 1>}
 head {type=Curve, loc=<0 0 0>, rot=<0 0 0 1>}
 dyn_ear_l_master {type=Curve, loc=<0 0 0>, rot=<0 0 0 1>}
 root {type=Curve, loc=<0 0 0>, rot=<0 0 0 1>}
 dyn_ear02_l {type=Curve, loc=<0 0 0>, rot=<0 0 0 1>}
 dyn_ear03_l {type=Curve, loc=<0 0 0>, rot=<0 0 0 1>}
 dyn_ear04_l {type=Curve, loc=<0 0 0>, rot=<0 0 0 1>}
 dyn_ear05_l {type=Curve, loc=<0 0 0>, rot=<0 0 0 1>}
 dyn_ear01_l {type=Curve, loc=<0 0 0>, rot=<0 0 0 1>}
 flipbook_face_L_eye_index {type=Curve, loc=<0 0 0>, rot=<0 0 0 1>}
 flipbook_face_R_eye_index {type=Curve, loc=<0 0 0>, rot=<0 0 0 1>}
 flipbook_face_L_brow_index {type=Curve, loc=<0 0 0>, rot=<0 0 0 1>}
 flipbook_face_R_brow_index {type=Curve, loc=<0 0 0>, rot=<0 0 0 1>}
 flipbook_face_mouth_index {type=Curve, loc=<0 0 0>, rot=<0 0 0 1>}
 flipbook_face_bypass_mapping {type=Curve, loc=<0 0 0>, rot=<0 0 0 1>}
 flipbook_face_is_idling {type=Curve, loc=<0 0 0>, rot=<0 0 0 1>}
]
variables [
 isCrouchingCR {type=bool, default=False}
 isTargetingCR {type=bool, default=False}
]
