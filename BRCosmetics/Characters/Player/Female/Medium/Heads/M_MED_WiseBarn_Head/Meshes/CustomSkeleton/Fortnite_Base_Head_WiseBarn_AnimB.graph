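# Post-process-style AnimBlueprint for the WiseBarn female head cosmetic, adding secondary motion to hay, ribbon and medal bones.
# It copies the body mesh pose, layers clavicle and upper-arm branches from it, and constrains base_ribbon between two spine bones.
# Spring-based AnimDynamics, with spine spherical limits for ribbons and medal, feeds a Control Rig receiving crouch and targeting states.
# The EventGraph computes state-based gravity and changes a velocity-scale variable on zipline events.
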AnimBlueprint /BRCosmetics/Characters/Player/Female/Medium/Heads/M_MED_WiseBarn_Head/Meshes/CustomSkeleton/Fortnite_Base_Head_WiseBarn_AnimB
target_skeleton=&/Fortnite_Base_Head_WiseBarn_Skeleton
graphs [
 AnimGraph {nodes [
   n1 = Layered blend per bone(BasePose=Input Pose(), BlendPoses_0=Copy Pose From Mesh(SourceMeshComponent=Get Body_Mesh())) {layerSetup [{branchFilters [
       {boneName=dfrm_clavicle_r}
       {boneName=dfrm_upperarm_r}
       {boneName=dfrm_upperarm_twist_01_r}
       {boneName=dfrm_clavicle_l}
       {boneName=dfrm_upperarm_l}
       {boneName=dfrm_upperarm_twist_01_l}
      ]}]}
   n2 = "Constraint - base_ribbon (PP)"!(ComponentPose=Local To Component(LocalPose=@n1), Alpha=1, ConstraintWeights_0=0.5639, ConstraintWeights_1:float=0.4361) -> (Pose) {
    boneToModify {boneName=base_ribbon}
    constraintSetup [{targetBone {boneName=spine_04}}, {targetBone {boneName=spine_05}}]
    constraintWeights [1,1]
   }
   n3 = Anim Dynamics - Bone: dyn_hay_01!(ComponentPose=@n2, Alpha=1, GravityOverride:Vector="0,0,0", ComponentLinearVelScale:Vector="0,0,0", ExternalForce="0,0,0") -> (Pose) {
    angularDampingOverride=1
    angularSpringConstant=11000
    bAngularSpring=true
    bOverrideAngularDamping=true
    bUseGravityOverride=true
    gravityScale=0.35
    simulationSpace=RootRelative
    boundBone {boneName=dyn_hay_01}
    componentAppliedLinearAccClamp=<9500 9500 9500>
    componentLinearAccScale=<0.5 0.5 0.5>
    componentLinearVelScale=<1 1 1>
    gravityOverride {z=-980}
    physicsBodyDefinitions [{
      boundBone {boneName=dyn_hay_01}
      boxExtents=<20 30 40>
      constraintSetup {angularLimitsMax=<5 15 8>, angularLimitsMin=<-5 -15 -20>, angularTarget=<1 0 0>}
      localJointOffset=<2 0 0>
     }]
   }
   n4 = Anim Dynamics - Bone: dyn_hay_02!(ComponentPose=@n3, Alpha=1, GravityOverride:Vector="0,0,0", ComponentLinearVelScale:Vector="0,0,0", ExternalForce="0,0,0") -> (Pose) {
    angularDampingOverride=1
    angularSpringConstant=20000
    bAngularSpring=true
    bOverrideAngularDamping=true
    bUseGravityOverride=true
    gravityScale=0.35
    simulationSpace=RootRelative
    boundBone {boneName=dyn_hay_02}
    componentLinearAccScale=<0.5 0.5 0.5>
    componentLinearVelScale=<1 1 1>
    gravityOverride {z=-980}
    physicsBodyDefinitions [{
      boundBone {boneName=dyn_hay_02}
      boxExtents=<20 30 40>
      constraintSetup {angularLimitsMax=<5 15 8>, angularLimitsMin=<-5 -15 -18>, angularTarget=<1 0 0>}
      localJointOffset=<3 0 0>
     }]
   }
   n5 = Anim Dynamics - Bone: dyn_hay_03!(ComponentPose=@n4, Alpha=1, GravityOverride:Vector="0,0,0", ComponentLinearVelScale:Vector="0,0,0", ExternalForce="0,0,0") -> (Pose) {
    angularDampingOverride=1
    angularSpringConstant=20000
    bAngularSpring=true
    bOverrideAngularDamping=true
    bUseGravityOverride=true
    gravityScale=0.35
    simulationSpace=RootRelative
    boundBone {boneName=dyn_hay_03}
    componentLinearAccScale=<1 1 1>
    componentLinearVelScale=<1 1 1>
    gravityOverride {z=-980}
    physicsBodyDefinitions [{
      boundBone {boneName=dyn_hay_03}
      boxExtents=<20 30 30>
      constraintSetup {angularLimitsMax=<5 10 10>, angularLimitsMin=<-5 -10 -15>, angularTarget=<1 0 0>}
      localJointOffset=<4 0 0>
     }]
   }
   n6 = Anim Dynamics - Bone: dyn_hay_04!(ComponentPose=@n5, Alpha=1, GravityOverride:Vector="0,0,0", ComponentLinearVelScale:Vector="0,0,0", ExternalForce="0,0,0") -> (Pose) {
    angularDampingOverride=1
    angularSpringConstant=20000
    bAngularSpring=true
    bOverrideAngularDamping=true
    bUseGravityOverride=true
    gravityScale=0.35
    simulationSpace=RootRelative
    boundBone {boneName=dyn_hay_04}
    componentLinearAccScale=<1 1 1>
    componentLinearVelScale=<1 1 1>
    gravityOverride {z=-980}
    physicsBodyDefinitions [{
      boundBone {boneName=dyn_hay_04}
      boxExtents=<15 22 22>
      constraintSetup {angularLimitsMax=<5 10 10>, angularLimitsMin=<-5 -10 -15>, angularTarget=<1 0 0>}
      localJointOffset=<4 0 0>
     }]
   }
   n7 = Anim Dynamics - Bone: dyn_ribbon_l!(ComponentPose=@n6, Alpha=1, GravityOverride:Vector="0,0,-980", ComponentLinearVelScale:Vector="1,1,1", ExternalForce="0,0,0") -> (Pose) {
    angularDampingOverride=1
    angularSpringConstant=11000
    bAngularSpring=true
    bOverrideAngularDamping=true
    bUseGravityOverride=true
    bUseSphericalLimits=true
    gravityScale=0.35
    simulationSpace=RootRelative
    boundBone {boneName=dyn_ribbon_l}
    componentAppliedLinearAccClamp=<9500 9500 9500>
    componentLinearAccScale=<0.5 0.5 0.5>
    componentLinearVelScale=<1 1 1>
    gravityOverride {z=-980}
    physicsBodyDefinitions [{
      collisionType=CustomSphere
      sphereCollisionRadius=1.5
      boundBone {boneName=dyn_ribbon_l}
      boxExtents=<30 40 40>
      constraintSetup {angularLimitsMax=<1 1 15>, angularLimitsMin=<-1 -10 -15>, angularTarget=<1 0 0>}
      localJointOffset=<4 0 0>
     }]
    sphericalLimits [{limitRadius=10, drivingBone {boneName=spine_05}, sphereLocalOffset=<-5 10 -6>}]
   }
   n8 = Anim Dynamics - Bone: dyn_medal!(ComponentPose=@n7, Alpha=1, GravityOverride:Vector="0,0,-980", ComponentLinearVelScale:Vector="1,1,1", ExternalForce="0,0,0") -> (Pose) {
    angularDampingOverride=1
    angularSpringConstant=11000
    bAngularSpring=true
    bOverrideAngularDamping=true
    bUseGravityOverride=true
    bUseSphericalLimits=true
    gravityScale=0.35
    simulationSpace=RootRelative
    boundBone {boneName=dyn_medal}
    componentAppliedLinearAccClamp=<9500 9500 9500>
    componentLinearAccScale=<0.5 0.5 0.5>
    componentLinearVelScale=<1 1 1>
    gravityOverride {z=-980}
    physicsBodyDefinitions [{
      collisionType=CustomSphere
      sphereCollisionRadius=1
      boundBone {boneName=dyn_medal}
      boxExtents=<30 40 40>
      constraintSetup {angularLimitsMax=<1 1 15>, angularLimitsMin=<-1 -20 -15>, angularTarget=<1 0 0>}
      localJointOffset=<4 0 0>
     }]
    sphericalLimits [{limitRadius=10.5, drivingBone {boneName=spine_05}, sphereLocalOffset=<-5 10 0>}]
   }
   n9 = Anim Dynamics - Bone: dyn_ribbon_r!(ComponentPose=@n8, Alpha=1, GravityOverride:Vector="0,0,-980", ComponentLinearVelScale:Vector="1,1,1", ExternalForce="0,0,0") -> (Pose) {
    angularDampingOverride=1
    angularSpringConstant=11000
    bAngularSpring=true
    bOverrideAngularDamping=true
    bUseGravityOverride=true
    bUseSphericalLimits=true
    gravityScale=0.35
    simulationSpace=RootRelative
    boundBone {boneName=dyn_ribbon_r}
    componentAppliedLinearAccClamp=<9500 9500 9500>
    componentLinearAccScale=<0.5 0.5 0.5>
    componentLinearVelScale=<1 1 1>
    gravityOverride {z=-980}
    physicsBodyDefinitions [{
      collisionType=CustomSphere
      sphereCollisionRadius=1.5
      boundBone {boneName=dyn_ribbon_r}
      boxExtents=<30 40 40>
      constraintSetup {angularLimitsMax=<1 10 15>, angularLimitsMin=<-1 -1 -15>, angularTarget=<1 0 0>}
      localJointOffset=<4 0 0>
     }]
    sphericalLimits [{limitRadius=10, drivingBone {boneName=spine_05}, sphereLocalOffset=<-5 10 6>}]
   }
   n10 = Control Rig!(Source=Component To Local(ComponentPose=@n9), Alpha, isCrouchingCR:bool=Get bIsCrouching(), isTargetingCR:bool=Get bIsTargeting()) -> (Pose) {controlRigAssetReference {blueprintRigClass=/BRCosmetics/Characters/Player/Female/Medium/Heads/M_MED_WiseBarn_Head/Meshes/F_MED_WiseBarn_Head_CtrlRig.F_MED_WiseBarn_Head_CtrlRig_C}}
   Output Pose(Result=@n10)
  ]}
 EventGraph {nodes [
   Event OnCharacterCustomizationLoaded()
   > n1 = Cast To PlayerPawn_Athena(Object=TryGetPawnOwner())
   > Set Body_Mesh(Body_Mesh=GetSkeletalMeshForPartType(self=@n1))
   Event BlueprintUpdateAnimation()
   > n2 = Sequence()
   n3 = Event BlueprintProcessAnimInputEvent()
   n4 = IsNamed(self=@n3.AnimInputEvent, InEventName=aie.Player.Zipline.Entered)
   n5 = OR Boolean(A=Get bIsMoving2D(), B=Get bIsTacticalSprinting())
   n6 = Get bIsJumping()
   n7 = Branch(execute=@n2.then_1, Condition=OR Boolean(A=@n4, B=AND Boolean!(A=@n4, B=@n6, C:bool=@n5) -> (ReturnValue)))
   > Set Componen Linear Vel Scale(Componen Linear Vel Scale:Vector="0, 0, 0") -> (Output_Get:Vector)
   Branch(execute=@n7.else, Condition=IsNamed(self=@n3.AnimInputEvent, InEventName=aie.Player.Zipline.Exited))
   > Set Componen Linear Vel Scale(Componen Linear Vel Scale="0.3,0.3,0.3")
   n8 = "Gravity Override State Function (Parameter Asset) (Nativized)"(
    execute=@n2.then_0
    CharacterPartAnimInstance=Self-Reference()
    GameTimeInSeconds=GetGameTimeInSeconds()
    Parameters=/BRCosmetics/Characters/Player/Female/Medium/Bodies/M_MED_WiseBarn/Meshes/F_MED_Wisebarn_Parameters
   )
   > Set Gravity Override(Gravity Override=@n8)
  ]}
]
variables [
 Body_Mesh {type=SkeletalMeshComponent}
 Componen Linear Vel Scale {type=Vector, default=<0.3 0.3 0.3>}
 Gravity Override {type=Vector}
 Gravity Override Params_0 {
  type=FortGravityOverrideParameterSet
  default=(Skydive_GoingUp=(GravityOverride=()),Skydive_FloatingDown=(GravityOverride=()),Skydive_DivingDown=(GravityOverride=()),Skydive_Parachuting=(GravityOverride=()),OnGround_Standing=(GravityOverride=()),OnGround_Moving=(GravityOverride=()),OnGround_Moving_HighRotationRate=(GravityOverride=()),OnGround_Crouch_Moving=(GravityOverride=()),Emote_or_Melee=(GravityOverride=()),Swimming=(GravityOverride=()),Front_End=(GravityOverride=()),DBNO=(GravityOverride=()),DBNO_Carried=(GravityOverride=()),OnGround_Falling=(GravityOverride=()),Ziplining=(GravityOverride=()),Skydive_Surfing=(GravityOverride=()),Riding_Standing=(GravityOverride=()),Riding_Moving=(GravityOverride=()),Riding_Moving_Targeting=(GravityOverride=()),Tactical_Sprint=(GravityOverride=()),Rail_Grinding=(GravityOverride=()),)
 }
 RBAN Params_0 {
  type=FortRigidBodyAnimNodeParameterSet
  default=(Skydive_GoingUp=(RigidBodyAnimNode=()),Skydive_FloatingDown=(RigidBodyAnimNode=()),Skydive_DivingDown=(RigidBodyAnimNode=()),Skydive_Parachuting=(RigidBodyAnimNode=()),OnGround_Standing=(RigidBodyAnimNode=()),OnGround_Moving=(RigidBodyAnimNode=()),OnGround_Moving_HighRotationRate=(RigidBodyAnimNode=()),OnGround_Crouch_Moving=(RigidBodyAnimNode=()),Emote_or_Melee=(RigidBodyAnimNode=()),Swimming=(RigidBodyAnimNode=()),Front_End=(RigidBodyAnimNode=()),DBNO=(RigidBodyAnimNode=()),DBNO_Carried=(RigidBodyAnimNode=()),OnGround_Falling=(RigidBodyAnimNode=()),Ziplining=(RigidBodyAnimNode=()),Skydive_Surfing=(RigidBodyAnimNode=()),Riding_Standing=(RigidBodyAnimNode=()),Riding_Moving=(RigidBodyAnimNode=()),Riding_Moving_Targeting=(RigidBodyAnimNode=()),Tactical_Sprint=(RigidBodyAnimNode=()),Rail_Grinding=(RigidBodyAnimNode=()),)
 }
 Sim Space Settings {
  type=SimSpaceSettings
  default {
   dampingAlpha=1
   maxAngularAcceleration=10000
   maxAngularVelocity=10000
   maxLinearAcceleration=10000
   maxLinearVelocity=10000
   velocityScaleZ=1
   worldAlpha=0
   externalLinearDragV=<0 0 0>
  }
 }
]
